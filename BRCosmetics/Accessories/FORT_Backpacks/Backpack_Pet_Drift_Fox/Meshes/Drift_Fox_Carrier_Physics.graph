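# Physics asset for the Drift Fox pet backpack carrier, giving its hanging daggers and zippers box bodies for physical secondary motion.
# The root bone carries two large boxes as the pack body, with both zipper pulls constrained directly to it with tight twist limits.
# Left, middle and right daggers each hang from a small chain-start body, which rotates freely on the root while the dagger's joint to it is angle-limited.
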PhysicsAsset /BRCosmetics/Accessories/FORT_Backpacks/Backpack_Pet_Drift_Fox/Meshes/Drift_Fox_Carrier_Physics
bodies {
 dyn_chain_l_start1 [{shapeType=Box, extentX=1.01, extentY=1.01, extentZ=1.01}]
 dyn_chain_mid_start [{shapeType=Box, extentX=1.01, extentY=1.01, extentZ=1.01}]
 dyn_chain_r_start [{shapeType=Box, extentX=1.01, extentY=1.01, extentZ=1.01}]
 dyn_dagger_l [{shapeType=Box, extentX=14.9406, extentY=2.1229, extentZ=5.1119, center=<7.3551 0 0.6462>, rotation=<rot 6.1023 0 0>}]
 dyn_dagger_mid [{shapeType=Box, extentX=16.2593, extentY=2.1995, extentZ=5.3153, center=<6.941 0 0>}]
 dyn_dagger_r [{shapeType=Box, extentX=14.5282, extentY=2.1237, extentZ=5.1129, center=<6.7079 0 -0.3108>, rotation=<rot -4.7649 0 0>}]
 dyn_zipper_l [{shapeType=Box, extentX=2.1721, extentY=0.5185, extentZ=1.01, center=<1.9671 0 0>}]
 dyn_zipper_r [{shapeType=Box, extentX=2.3196, extentY=0.5589, extentZ=1.01, center=<1.9671 0 0>}]
 root [
  {shapeType=Box, extentX=33.1977, extentY=15.1137, extentZ=43.5028, center=<0.0001 -5.0119 -9.8982>}
  {shapeType=Box, extentX=28.0524, extentY=8.7194, extentZ=10.0204, center=<0 -19.9039 -23.1841>}
 ]
}
constraints [
 {
  bone1Name=dyn_chain_mid_start
  bone2Name=root
  swing1LimitDegrees=45
  swing1Motion=Free
  swing2LimitDegrees=45
  swing2Motion=Free
  twistLimitDegrees=45
  twistMotion=Free
 }
 {
  bone1Name=dyn_chain_r_start
  bone2Name=root
  swing1LimitDegrees=45
  swing1Motion=Free
  swing2LimitDegrees=45
  swing2Motion=Free
  twistLimitDegrees=45
  twistMotion=Free
 }
 {
  bone1Name=dyn_chain_l_start1
  bone2Name=root
  swing1LimitDegrees=45
  swing1Motion=Free
  swing2LimitDegrees=45
  swing2Motion=Free
  twistLimitDegrees=45
  twistMotion=Free
 }
 {bone1Name=dyn_zipper_l, bone2Name=root, swing1LimitDegrees=45, swing2LimitDegrees=45, twistLimitDegrees=5}
 {bone1Name=dyn_zipper_r, bone2Name=root, swing1LimitDegrees=45, swing2LimitDegrees=45, twistLimitDegrees=5}
 {bone1Name=dyn_dagger_mid, bone2Name=dyn_chain_mid_start, swing1LimitDegrees=65, swing2LimitDegrees=45, twistLimitDegrees=15}
 {bone1Name=dyn_dagger_r, bone2Name=dyn_chain_r_start, swing1LimitDegrees=65, swing2LimitDegrees=45, twistLimitDegrees=15}
 {bone1Name=dyn_dagger_l, bone2Name=dyn_chain_l_start1, swing1LimitDegrees=65, swing2LimitDegrees=45, twistLimitDegrees=15}
]
