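# Animation Blueprint for the Drift Fox pet backpack carrier skeleton that adds secondary motion to its dagger bones.
# The input pose runs through a RigidBody simulation in root-bone space, using the carrier physics asset, overridden gravity and clamped component acceleration.
# Six chained Look At nodes then aim paired helper bones at the middle, right and left dagger bones.
# The event graph holds only unconnected default nodes, so all behavior lives in the AnimGraph.
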AnimBlueprint /BRCosmetics/Accessories/FORT_Backpacks/Backpack_Pet_Drift_Fox/Meshes/Drift_Fox_Carrier_Skeleton_AnimBP
target_skeleton=&/Drift_Fox_Carrier_Skeleton
graphs [
 AnimGraph {nodes [
   AnimGraphNode_SubInput_0 = Input Pose()
   n1 = RigidBody(ComponentPose=Local To Component(LocalPose=@AnimGraphNode_SubInput_0)) {
    bOverrideWorldGravity=true
    overridePhysicsAsset=&/Drift_Fox_Carrier_Physics
    simulationSpace=BaseBoneSpace
    alphaBoolBlend {blendOption=Linear}
    baseBoneRef {boneName=root}
    componentAppliedLinearAccClamp=<2000 2000 2000>
    componentLinearAccScale=<1 1 1>
    overrideWorldGravity {z=-980}
   }
   n2 = Look At - Bone: dyn_dagger_mid_r(ComponentPose=@n1, Alpha=1) {
    bUseLookUpAxis=true
    alphaBoolBlend {blendOption=Linear}
    boneToModify {boneName=dyn_dagger_mid_r}
    lookAtLocation {x=0, z=-1}
    lookAtTarget {boneReference {boneName=dyn_dagger_mid}}
    lookAt_Axis {axis {y=0, z=-1}}
    lookUp_Axis {axis=<0 1>}
   }
   n3 = Look At - Bone: dyn_dagger_mid_r1(ComponentPose=@n2, Alpha=1) {
    alphaBoolBlend {blendOption=Linear}
    boneToModify {boneName=dyn_dagger_mid_r1}
    lookAtLocation {x=0, z=-1}
    lookAtTarget {boneReference {boneName=dyn_dagger_mid}}
    lookAt_Axis {axis {y=0, z=1}}
    lookUp_Axis {axis {x=-1}}
   }
   n4 = Look At - Bone: dyn_dagger_r_l(ComponentPose=@n3, Alpha=1) {
    bUseLookUpAxis=true
    alphaBoolBlend {blendOption=Linear}
    boneToModify {boneName=dyn_dagger_r_l}
    lookAtLocation {x=0, z=-1}
    lookAtTarget {boneReference {boneName=dyn_dagger_r}}
    lookAt_Axis {axis {y=0, z=-1}}
    lookUp_Axis {axis {y=1}}
   }
   n5 = Look At - Bone: dyn_dagger_r_r(ComponentPose=@n4, Alpha=1) {
    bUseLookUpAxis=true
    alphaBoolBlend {blendOption=Linear}
    boneToModify {boneName=dyn_dagger_r_r}
    lookAtLocation {x=0, z=1}
    lookAtTarget {boneReference {boneName=dyn_dagger_r}}
    lookAt_Axis {axis {y=0, z=1}}
    lookUp_Axis {axis {y=1}}
   }
   n6 = Look At - Bone: dyn_dagger_l_l(ComponentPose=@n5, Alpha=1) {
    bUseLookUpAxis=true
    alphaBoolBlend {blendOption=Linear}
    boneToModify {boneName=dyn_dagger_l_l}
    lookAtLocation {x=0, z=-1}
    lookAtTarget {boneReference {boneName=dyn_dagger_l}}
    lookAt_Axis {axis {y=0, z=-1}}
    lookUp_Axis {axis {y=1}}
   }
   n7 = Look At - Bone: dyn_dagger_l_r(ComponentPose=@n6, Alpha=1) {
    bUseLookUpAxis=true
    alphaBoolBlend {blendOption=Linear}
    boneToModify {boneName=dyn_dagger_l_r}
    lookAtLocation {x=0, z=1}
    lookAtTarget {boneReference {boneName=dyn_dagger_l}}
    lookAt_Axis {axis {y=0, z=1}}
    lookUp_Axis {axis {y=1}}
   }
   Output Pose(Result=Component To Local(ComponentPose=@n7)) {layerGroup=None}
  ]}
 EventGraph {nodes [
   TryGetPawnOwner()
   Event BlueprintUpdateAnimation()
  ]}
]
variables []
